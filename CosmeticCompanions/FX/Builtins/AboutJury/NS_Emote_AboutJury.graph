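# Niagara system for the AboutJury emote: confetti plus a liquid mesh with a late slurp ribbon, drops and splashes.
# Liquid follows a skeletal socket via a viscous pendulum, feeding tilt, speed and turbulence to material parameters.
# Ribbon lerps between two sampled points as timed drops and splashes spawn; front-end modifiers swap their offsets and colors.
# Confetti combines a timed burst with curve-driven spawning, cheer-scaled curl noise, collision and time-of-day tinting.
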NiagaraSystem /CosmeticCompanions/FX/Builtins/AboutJury/NS_Emote_AboutJury
dependencies {
 usedDataInterfaces [
  &NiagaraDataInterfaceCurve
  &NiagaraDataInterfaceColorCurve
  &NiagaraDataInterfaceArrayColor
  &NiagaraDataInterfaceSpriteRendererInfo
  &NiagaraDataInterfaceSkeletalMesh
 ]
 usedDynamicInputs []
 usedModules [
  &self:SystemSpawnScript.SystemScriptSource.SystemScriptGraph.NiagaraNodeAssignment_0.SetVariables_#1
  &SystemState
  &self:SystemSpawnScript.SystemScriptSource.SystemScriptGraph.NiagaraNodeAssignment_1.SetVariables_#2
 ]
}
emitters {
 Confetti001 {
  emitterUpdateScript [
   &EmitterState()
   &SpawnBurst_Instantaneous(Spawn Count=39, Spawn Group=1, Spawn Time=1.525, Use Spawn Probability=false, UseLoopCountLimit=false)
   &SpawnRate(
    SpawnRate=&FloatFromCurve(
     FloatCurve=DataInterface() {propertyValues {Curve {keys [
         <at 0 0 RCIM_Constant>
         <at 1.52 1 RCIM_Cubic>
         {arriveTangent=-0.3082, interpMode=RCIM_Cubic, leaveTangent=-0.3082, tangentMode=RCTM_User, time=8.7, value=0}
        ]}}}
     CurveIndex=$Emitter.LoopedAge
     Scale Curve=40
    )
    Use Spawn Probability=false
   )
  ]
  particleSpawnScript [
   &InitializeParticle(
    Color=&SelectLinearColorFromArray(
     Array Sampling Mode=::NewEnumerator0
     Direct Array Index=?
     Color Selection Array=DataInterface() {propertyValues {ColorData [<rgba 2 1.2806 0.1957 1>, <rgba 1.4307 5 4.3272 1>, <rgba 2 0.516 1.8186 1>, <rgba 0.4138 2 0.7293 1>, <rgba 2 0.5102 0.3955 1>]}}
     Random Seed=0
    )
    Lifetime Max=2.8
    Lifetime Mode=::NewEnumerator1
    Sprite Size Mode=::NewEnumerator1
    Uniform Sprite Size Max=6
    Uniform Sprite Size Min=3
    UsePositionOffset=false
    Write Material Random=true
   )
   &ShapeLocation(
    MaskSpawnGroup=true
    Non Uniform Scale=<1 1 0.5>
    Offset=<0 0 120>
    Offset Mode=::NewEnumerator0
    Override Seed=false
    Shape Primitive=::NewEnumerator2
    Surface Only=false
    UseOwnerScale=true
   )
   &ShapeLocation(
    Box Size=<29.8283 43.273 68.3453>
    MaskSpawnGroup=true
    Non Uniform Scale=<1 1 0.5>
    Offset=<0 0 140>
    Offset Mode=::NewEnumerator0
    Override Seed=false
    Shape Primitive=::NewEnumerator2
    Spawn Group Mask=1
    Surface Only=false
    UseOwnerScale=true
   ) {module=ShapeLocation001}
   &Color(Color=&MultiplyLinearColor(
     Linear Color=$Particles.Color
     Scale Factor=&MakeVector4_FromLinearColor(LinearColor=&FNTimeOfDayColor(
       Night Color=<rgba 0.3782 0.3614 0.4948 1>
       Morning Color=<rgba 1 0.7262 0.6963 1>
       Day Color=<rgba 1 0.9315 0.9012 1>
       Evening Color=<rgba 1 0.9069 0.7687 1>
      ))
    ))
   &AddVelocity(
    Module.Use Falloff Distance=false
    Origin Offset:NiagaraPosition=$Particles.ShapeLocation.ShapePosition
    Override Seed=false
    Velocity Mode=::NewEnumerator1
    Velocity Speed=&MaskFloatBySpawnGroup(Masked Float=0, Passthrough Float=&RandomRangeFloat(Minimum=80, Maximum=145))
   )
  ]
  particleUpdateScript [
   &Drag(
    Drag=&MaskFloatBySpawnGroup(
     Masked Float=3.5
     Passthrough Float=&FloatFromCurve(
      FloatCurve=DataInterface() {propertyValues {Curve {keys [<at 0 0 RCIM_Cubic>, {arriveTangent=31.2634, interpMode=RCIM_Cubic, leaveTangent=31.2634, tangentMode=RCTM_User, time=0.1, value=1}]}}}
      Scale Curve=&RandomRangeFloat(Minimum=4, Maximum=7)
     )
    )
    Use Linear Drag=true
    Use Rotational Drag=true
   )
   &GravityForce(Gravity=&RandomRangeVector(Minimum=<0 0 -110>, Maximum=<0 0 -400>))
   &CurlNoiseForce(
    Bias Noise=false
    Noise Frequency=10
    Noise Strength=&Multiply_Float(
     A=&RandomRangeFloat(Minimum=30, Maximum=50)
     B=&Lerp_Float(A=3, B=10, Alpha=$User.CrowdCheerCurve)
    )
    Pan Noise=true
    Pan Noise Field=<0.2 0.2 0.2>
   )
   &Collision(
    Age Colliding Particles=false
    KillOnCollision=false
    Module.Control Rest State Via Surface Normal=false
    Module.EnableMaxCollisionCount=false
    Module.Randomize Collision Normal Vector=false
    Restitution=0
   )
   &SolveForcesAndVelocity(Clamp Velocity=false, Limit Acceleration=false)
   &ParticleState()
   &ScaleSpriteSize(Uniform Curve Sprite Scale=DataInterface() {propertyValues {Curve {keys [
        {interpMode=RCIM_Cubic, leaveTangent=12.0846, tangentMode=RCTM_Break, time=0, value=0}
        {interpMode=RCIM_Cubic, tangentMode=RCTM_Break, time=0.1, value=1}
        {arriveTangent=-2.7226, tangentMode=RCTM_User, time=1, value=0}
       ]}}})
   /SetVariables(Particles.SubImageIndex=&FloatFromCurve(
     FloatCurve=DataInterface() {propertyValues {Curve {keys [<at 0 0>,<at 1 1>]}}}
     Scale Curve=&UniformAOrBFloat(A=90, B=56, Distribution Weight=0.1, Evaluation Type=::NewEnumerator0, Recalculate Random Each Loop=true, Override Seed=false)
    )) {isSetParameters=true}
   &ScaleColor(
    Scale Mode=1
    Scale RGBA=&MakeVector4_FromLinearColor(LinearColor=&FNTimeOfDayColor(
      Night Color=&ColorFromCurve(ColorCurve=DataInterface() {propertyValues {
         AlphaCurve {keys [<at 0 1 RCIM_Cubic>]}
         BlueCurve {keys [<at 0 1 RCIM_Cubic>,<at 0.4559 1>,<at 0.4752 1000>,<at 0.4979 1>,<at 1 1 RCIM_Cubic>]}
         GreenCurve {keys [<at 0 1 RCIM_Cubic>,<at 0.4559 1>,<at 0.4752 1000>,<at 0.4979 1>,<at 1 1 RCIM_Cubic>]}
         RedCurve {keys [<at 0 1 RCIM_Cubic>,<at 0.4559 1>,<at 0.4752 1000>,<at 0.4979 1>,<at 1 1 RCIM_Cubic>]}
        }})
      Morning Color=&ColorFromCurve(ColorCurve=DataInterface() {propertyValues {
         AlphaCurve {keys [<at 0 1 RCIM_Cubic>]}
         BlueCurve {keys [<at 0 1 RCIM_Cubic>,<at 0.4559 1>,<at 0.4752 1000>,<at 0.4979 1>,<at 1 1 RCIM_Cubic>]}
         GreenCurve {keys [<at 0 1 RCIM_Cubic>,<at 0.4559 1>,<at 0.4752 1000>,<at 0.4979 1>,<at 1 1 RCIM_Cubic>]}
         RedCurve {keys [<at 0 1 RCIM_Cubic>,<at 0.4559 1>,<at 0.4752 1000>,<at 0.4979 1>,<at 1 1 RCIM_Cubic>]}
        }})
      Day Color=&ColorFromCurve(ColorCurve=DataInterface() {propertyValues {
         AlphaCurve {keys [<at 0 1 RCIM_Cubic>]}
         BlueCurve {keys [<at 0 1 RCIM_Cubic>,<at 0.4559 1>,<at 0.4752 500>,<at 0.4979 1>,<at 1 1 RCIM_Cubic>]}
         GreenCurve {keys [<at 0 1 RCIM_Cubic>,<at 0.4559 1>,<at 0.4752 500>,<at 0.4979 1>,<at 1 1 RCIM_Cubic>]}
         RedCurve {keys [<at 0 1 RCIM_Cubic>,<at 0.4559 1>,<at 0.4752 500>,<at 0.4979 1>,<at 1 1 RCIM_Cubic>]}
        }})
      Evening Color=&ColorFromCurve(ColorCurve=DataInterface() {propertyValues {
         AlphaCurve {keys [<at 0 1 RCIM_Cubic>]}
         BlueCurve {keys [<at 0 1 RCIM_Cubic>,<at 0.4559 1>,<at 0.4752 1000>,<at 0.4979 1>,<at 1 1 RCIM_Cubic>]}
         GreenCurve {keys [<at 0 1 RCIM_Cubic>,<at 0.4559 1>,<at 0.4752 1000>,<at 0.4979 1>,<at 1 1 RCIM_Cubic>]}
         RedCurve {keys [<at 0 1 RCIM_Cubic>,<at 0.4559 1>,<at 0.4752 1000>,<at 0.4979 1>,<at 1 1 RCIM_Cubic>]}
        }})
     ))
    ScaleRGBA=true
   )
  ]
  rendererClasses [&NiagaraSpriteRendererProperties]
  renderers [{rendererClass=&NiagaraSpriteRendererProperties}]
 }
 Flare {rendererClasses [], renderers []}
 GlassDrops001 {
  emitterUpdateScript [
   &EmitterState(Life Cycle Mode=::NewEnumerator1, Loop Behavior=::NewEnumerator1, Loop Delay=8.15, Loop Duration=0.24, UseLoopDelay=true)
   &SpawnRate(SpawnRate=&FN_FrontEndFloatModifier(Gameplay=96, Front End=145), Use Spawn Probability=false)
   &SpawnBurst_Instantaneous(Spawn Count=15, Use Spawn Probability=false, UseLoopCountLimit=false)
  ]
  particleSpawnScript [
   &InitializeParticle(
    Color=&MakeLinearColorFromVectorAndFloat(
     "Vector (RGB)"=&FN_FrontEndVector3Modifier(
      Gameplay=&MakeVectorFromLinearColorRGB(Color=&RandomRangeLinearColor(Minimum=<rgba 0.0773 0.9331 1 1>, Maximum=<rgba 0.0838 1 0.5546 1>))
      Front End=&MakeVectorFromLinearColorRGB(Color=&RandomRangeLinearColor(Minimum=<rgba 0.1713 1 0.0885 1>, Maximum=<rgba 0.0417 1 0.2964 1>))
     )
     "Float (Alpha)"=1
    )
    Lifetime=&FN_FrontEndFloatModifier(
     Gameplay=&RandomRangeFloat(Minimum=0.1, Maximum=0.225)
     Front End=&RandomRangeFloat(Minimum=0.1, Maximum=0.225)
    )
    Sprite Size Mode=::NewEnumerator1
    Uniform Sprite Size Max=1.6
    Uniform Sprite Size Min=0.9
    UsePositionOffset=false
    Write Material Random=true
   )
   &SocketLocation(SamplePosOffset=false, Socket Index=1, Socket Reader=$System.Socket Reader, Socket Sampling Mode=::NewEnumerator3, VelocitySpeedLimit=false) {enabled=false}
   &SkeletalMeshLocation(
    Bone / Socket Index=1
    Bone Sampling Mode=::NewEnumerator4
    MaskSpawnGroup=false
    Sampled Position Offset=&FN_FrontEndVector3Modifier(Gameplay=<-4 4 -2>, Front End=<0 0 0>)
    Sampled Position Offset Coordinate Space:ENiagaraSkelMeshTransforms=::NewEnumerator2
    Skeletal Mesh=$System.Skeletal Mesh Reader
   ) {module=SkeletalMeshLocation001}
   &ShapeLocation(
    MaskSpawnGroup=false
    Offset=<-2 4 -8>
    Offset Mode=::NewEnumerator0
    Override Seed=false
    Shape Origin=$Particles.SkeletalMeshLocation001.SampledPosition
    Sphere Radius=2
    UseOwnerScale=true
   )
   &AddVelocity(
    Module.Use Falloff Distance=false
    Origin Offset=&AddVectorToPosition(Position=<-2 4 -8>, Vector=<0 0 -10>)
    Override Seed=false
    Velocity Mode=::NewEnumerator1
    Velocity Origin=$Particles.SkeletalMeshLocation001.SampledPosition
    Velocity Speed=&RandomRangeFloat(Minimum=25, Maximum=40)
   )
   &CameraOffset(Camera Offset Amount=&FN_FrontEndFloatModifier(Gameplay=0, Front End=&RandomRangeFloat(Minimum=12, Maximum=20))) {enabled=false}
  ]
  particleUpdateScript [
   &ParticleState()
   &ScaleColor(
    Scale RGB=&Multiply_VectorByFloat(
     Vector=&MakeVectorFromLinearColorRGB(Color=&FNTimeOfDayColor(Night Color=<rgba 0.3229 0.661 1 1>, Morning Color=<rgba 1 0.7745 0.7008 1>, Day Color=<rgba 1 1 1 1>, Evening Color=<rgba 1 1 1 1>))
     Float=&RandomRangeFloat(Minimum=1, Maximum=1.75)
    )
    ScaleA=true
    ScaleRGB=true
   )
   &GravityForce(Gravity=&RandomRangeVector(Minimum=<0 0 -850>, Maximum=<0 0 -1200>))
   &SolveForcesAndVelocity(Clamp Velocity=false, Limit Acceleration=false)
   &ScaleSpriteSizeBySpeed(Max Scale Factor=<1.6 5>, Min Scale Factor=<0.8 3>)
   &ScaleSpriteSize(Uniform Curve Sprite Scale=DataInterface() {propertyValues {Curve {
       defaultValue=0
       keys [
        {arriveTangent=9.0725, interpMode=RCIM_Cubic, leaveTangent=9.0725, tangentMode=RCTM_User, time=0, value=0}
        {arriveTangent=0.2183, interpMode=RCIM_Cubic, leaveTangent=0.2183, tangentMode=RCTM_User, time=0.1408, value=0.9847}
        {arriveTangent=-3.9306, interpMode=RCIM_Cubic, leaveTangent=-3.9306, tangentMode=RCTM_User, time=1, value=0}
       ]
      }}})
  ]
  rendererClasses [&NiagaraSpriteRendererProperties]
  renderers [{rendererClass=&NiagaraSpriteRendererProperties}]
 }
 Liquid001 {
  emitterUpdateScript [
   &EmitterState()
   &SpawnBurst_Instantaneous(Use Spawn Probability=false, UseLoopCountLimit=false)
   /SetVariables(Emitter.Orientation:Quat4f=$Engine.Owner.Rotation) {enabled=false, isSetParameters=true}
   /SetVariables(Emitter.Position:Vector3f=/Niagara/DynamicInputs/Helpers/SimulationPosition()) {isSetParameters=true}
   /SetVariables(
    Emitter.RotationVelocity:Vector3f=&Divide_Vector(
     A:Vector3f=&Subtract_Vector(
      A=&RotateVectorByQuaternion(Vector To Rotate=<0 0 1>, Quaternion=$Emitter.Orientation)
      B=&RotateVectorByQuaternion(Vector To Rotate=<0 0 1>, Quaternion=$Emitter.Previous.Orientation)
     )
     B:Vector3f=&VectorFromFloat(Value=$Engine.DeltaTime)
    )
    Emitter.Velocity:Vector3f=&Divide_Vector(
     A:Vector3f=&Subtract_Vector(A=$Emitter.Position, B=$Emitter.Previous.Position)
     B:Vector3f=&VectorFromFloat(Value=$Engine.DeltaTime)
    )
    Emitter.Viscosity=0.6
   ) {isSetParameters=true}
   /SetVariables(Emitter.InheritSourceMovement:NiagaraFloat=&Lerp_Float(A=0, B=1, Alpha=$Emitter.Viscosity)) {isSetParameters=true}
  ]
  particleSpawnScript [&InitializeParticle(
    Color=&FNTimeOfDayColor(Night Color=<rgba 0.281 0.3676 0.5955 1>, Morning Color=<rgba 1 0.73 0.6627 1>, Day Color=<rgba 1 1 1 1>, Evening Color=<rgba 1 1 1 1>)
    Lifetime=9.4
    Mesh Scale Mode=::NewEnumerator3
    Mesh Uniform Scale=$User.LiquidScale
    UsePositionOffset=false
    Write Material Random=true
   )]
  particleUpdateScript [
   &ParticleState()
   &SkeletalMeshLocation(Bone Sampling Mode=::NewEnumerator4, MaskSpawnGroup=false, Position Sampling=::NewEnumerator2, Skeletal Mesh=$System.Skeletal Mesh Reader)
   &SocketLocation(
    SamplePosOffset=true
    Sampled Position Offset=<0 0 -9.3>
    Socket Index=1
    Socket Reader=$System.Socket Reader
    Socket Sampling Mode=::NewEnumerator3
    VelocitySpeedLimit=false
   ) {enabled=false}
   /SetVariables(Particles.Position_Origin:NiagaraPosition=$Particles.SkeletalMeshLocation.SampledPosition) {isSetParameters=true}
   &PendulumSetup(
    Module.Enable Spring Driver=true
    Pendulum Length:NiagaraFloat=&Multiply_Float(A=16, B=$User.LiquidScale)
    Pendulum Pivot:NiagaraPosition=$Particles.Position_Origin
    Pendulum Rest Axis:Vector3f=&NormalizeVector(
     Vector To Normalize=&Add_Vector(
      A=<0 0 -1>
      B=&Lerp_Vector(A=<0 0 0>, B=&TransformVector(Vector=<0 0 -1>, Destination Space=::Simulation), Alpha=0.5)
     )
     Safe Normalize=true
     Safe Fallback Threshold=0
     Safe Fallback Vector=<0 0 0.001>
    )
   )
   &GravityForce(Gravity=&Lerp_Vector(A=<0 0 -980>, B=<0 0 -19.6>, Alpha=$Emitter.Viscosity))
   &Drag(Drag=&Lerp_Float(A=0, B=4, Alpha=$Emitter.Viscosity), Use Linear Drag=true, Use Rotational Drag=false)
   &InheritSourceMovement(
    Applied Position Scale=&VectorFromFloat(Value=$Emitter.InheritSourceMovement)
    ApplyForce=false
    ApplyPosition=true
    ApplyPositionFalloff=false
    Source Velocity:Vector3f=$Emitter.Velocity
   )
   &SolveForcesAndVelocity(Clamp Velocity=false, Limit Acceleration=false)
   &PendulumConstraint()
   &CalculateAccurateVelocity(Previous Position:NiagaraPosition=&Add_Vector(
     A=Linked() {linkedVariable {name=Particles.Previous.Position, type=NiagaraPosition}}
     B=&Multiply_VectorByFloat(Vector=$Emitter.Velocity, Float=&Multiply_Float(A=$Emitter.InheritSourceMovement, B=$Engine.DeltaTime))
    ))
   /Niagara/Modules/Update/Orientation/V2/OrientMeshToVector(
    Look At Direction:Vector3f=&NormalizeVector(
     Vector To Normalize=&Subtract_Vector(
      A=Linked() {linkedVariable {name=Particles.Position_Origin, type=NiagaraPosition}}
      B=Linked() {linkedVariable {name=Particles.Position, type=NiagaraPosition}}
     )
     Safe Normalize=true
     Safe Fallback Threshold=0
     Safe Fallback Vector=<0 0 0.001>
    )
    Override Quat=false
    Reference Axis Vector:Vector3f=<0 0 1>
    Up Axis Direction:Vector3f=<0 0 1>
    UseOptionalSideAxis=false
    UseOptionalUpAxis=false
   )
   &InterpolateOverTime(
    Rate Of Change=2
    Target Float Value:NiagaraFloat=&Max_Float(
     A:NiagaraFloat=&RemapRange(
      Input Value To Remap=&DotProduct(
       Vector A=<0 0 -1>
       Vector B=&ReturnMeshOrientationAxis(Mesh Orientation Axis=::ZAxis, Mesh Orientation Quat=$Particles.MeshOrientation)
      )
      Input Min=-1
      Input Max=0
      Output Min=0
      Output Max=1
     )
     B:NiagaraFloat=&RemapRange(
      Input Value To Remap=&Divide_Float(A=&Subtract_Float(A=$Particles.DistanceTraveled, B=$Particles.Previous.DistanceTraveled), B=$User.LiquidScale)
      Input Min=2
      Input Max=4
      Output Min=0
      Output Max=1
     )
    )
   )
   /SetVariables(Particles.LiquidSpeed:NiagaraFloat=&ClampFloat(Float=$Particles.InterpolateOverTime.InterpolateOverTime_Float.Moving Average)) {isSetParameters=true}
   &Timeline(Play=&SetBoolByFloatComparison(A=$Particles.LiquidSpeed, B=0.333), Play Rate=0, Timeline End=10)
   /SetVariables(Particles.Turbulence:NiagaraFloat=$Particles.Timeline.NormalizedTime) {isSetParameters=true}
   /Niagara/Modules/Update/Utility/IncrementOverTime(Rate of Change:NiagaraFloat=&Lerp_Float(A=1, B=1.25, Alpha=$Particles.LiquidSpeed))
   /SetVariables(
    Particles.DynamicMaterialParameter:Vector4f=&MakeFloat4FromVectorAndFloat(
     "Vector (RGB)"=&ReturnMeshOrientationAxis(Mesh Orientation Axis=::ZAxis, Mesh Orientation Quat=$Particles.MeshOrientation)
     "Float (Alpha)"=&FloatFromCurve(
      FloatCurve=DataInterface() {propertyValues {Curve {
         defaultValue=1
         keys [
          {arriveTangent=0.0162, interpMode=RCIM_Cubic, leaveTangent=0.0162, tangentMode=RCTM_User, time=0, value=0}
          {arriveTangent=0.0048, interpMode=RCIM_Cubic, leaveTangent=0.0048, tangentMode=RCTM_User, time=0.3, value=1}
          {
           arriveTangent=0.0012
           arriveTangentWeight=0.6554
           interpMode=RCIM_Cubic
           leaveTangentWeight=0.1301
           tangentMode=RCTM_Break
           time=0.82
           value=0.9918
          }
          {arriveTangent=0.0126, interpMode=RCIM_Cubic, leaveTangent=0.0126, tangentMode=RCTM_User, time=1.25, value=0.6}
          {arriveTangent=0.0085, interpMode=RCIM_Cubic, leaveTangent=0.0085, tangentMode=RCTM_User, time=8.3, value=0.6}
          <at 8.6 0 RCIM_Cubic>
         ]
        }}}
      CurveIndex=$Emitter.LoopedAge
     )
    )
    Particles.DynamicMaterialParameter1:Vector4f=&MakeFloat4FromVectorAndFloat(
     "Vector (RGB)"=&Multiply_VectorByFloat(
      Vector=&ReturnMeshOrientationAxis(Mesh Orientation Axis=::XAxis, Mesh Orientation Quat=$Particles.MeshOrientation)
      Float=1
     )
     "Float (Alpha)"=$Particles.IncrementOverTime.Counter
    )
    Particles.DynamicMaterialParameter2:Vector4f=&MakeVector4(X=$Particles.LiquidSpeed, Y=$Particles.Turbulence, Z=0, W=0)
   ) {isSetParameters=true}
  ]
  rendererClasses [&NiagaraMeshRendererProperties]
  renderers [{rendererClass=&NiagaraMeshRendererProperties, rendererIndex=0}]
 }
 SlurpJuice {
  emitterUpdateScript [
   &EmitterState()
   &SpawnRate(
    SpawnRate=&FloatFromCurve(
     FloatCurve=DataInterface() {propertyValues {Curve {keys [
         <at 7.875 0 RCIM_Cubic>
         {arriveTangent=0.1281, interpMode=RCIM_Cubic, leaveTangent=0.1281, tangentMode=RCTM_User, time=8.075, value=1}
         {arriveTangent=-0.1421, interpMode=RCIM_Cubic, leaveTangent=-0.1421, tangentMode=RCTM_User, time=8.175, value=1}
         <at 8.375 0 RCIM_Cubic>
        ]}}}
     CurveIndex=$Emitter.Age
     Scale Curve=60
    )
    Use Spawn Probability=false
   )
   /SetVariables(
    Emitter.LiquidA:LinearColor=&MakeLinearColorFromVectorAndFloat(
     "Vector (RGB)"=&FN_FrontEndVector3Modifier(
      Gameplay=&MakeVectorFromLinearColorRGB(Color=<rgba 0 0.7216 1 1>)
      Front End=&MakeVectorFromLinearColorRGB(Color=<rgba 1 1 1 1>)
     )
     "Float (Alpha)"=1
    )
    Emitter.LiquidB:LinearColor=&MakeLinearColorFromVectorAndFloat(
     "Vector (RGB)"=&FN_FrontEndVector3Modifier(
      Gameplay=&MakeVectorFromLinearColorRGB(Color=<rgba 0.1028 1 0.6115 1>)
      Front End=&MakeVectorFromLinearColorRGB(Color=<rgba 1 1 1 1>)
     )
     "Float (Alpha)"=1
    )
   ) {isSetParameters=true}
  ]
  particleSpawnScript [
   &InitializeParticle(
    Lifetime=0.2
    Position Offset=<-1.2 0 -2>
    Ribbon Width=4
    Ribbon Width Mode=::NewEnumerator1
    UsePositionOffset=false
    Write Material Random=true
   )
   &CameraOffset(Camera Offset Amount=-5)
  ]
  particleUpdateScript [
   &SkeletalMeshLocation(
    Bone Sampling Mode=::NewEnumerator4
    MaskSpawnGroup=false
    Position Sampling=::NewEnumerator2
    Sampled Position Offset=&FN_FrontEndVector3Modifier(Gameplay=<2 0 3>, Front End=<-2 0 -2>)
    Sampled Position Offset Coordinate Space:ENiagaraSkelMeshTransforms=::NewEnumerator2
    Skeletal Mesh=$System.Skeletal Mesh Reader
   )
   &SkeletalMeshLocation(
    Bone / Socket Index=1
    Bone Sampling Mode=::NewEnumerator4
    MaskSpawnGroup=false
    Position Sampling=::NewEnumerator2
    Sampled Position Offset=<-6 0 -8>
    Sampled Position Offset Coordinate Space:ENiagaraSkelMeshTransforms=::NewEnumerator2
    Skeletal Mesh=$System.Skeletal Mesh Reader
   ) {module=SkeletalMeshLocation001}
   &SolveForcesAndVelocity(Clamp Velocity=false, Limit Acceleration=false)
   &ParticleState()
   &ScaleColor(
    Scale Alpha=&FloatFromCurve(FloatCurve=DataInterface() {propertyValues {Curve {keys [
         {interpMode=RCIM_Cubic, tangentMode=RCTM_User, time=0, value=0}
         {arriveTangent=11.7094, interpMode=RCIM_Cubic, leaveTangent=11.7094, tangentMode=RCTM_User, time=0.2, value=1}
         {arriveTangent=0.0062, interpMode=RCIM_Cubic, leaveTangent=0.0062, tangentMode=RCTM_User, time=0.21, value=1}
        ]}}})
    Scale RGB=&FN_FrontEndVector3Modifier(
     Gameplay=<1 1 1>
     Front End=&MakeVectorFromLinearColorRGB(Color=&ColorFromCurve(ColorCurve=DataInterface() {propertyValues {
         AlphaCurve {keys [<at 0 1 RCIM_Cubic>]}
         BlueCurve {keys [<at 0 1 RCIM_Cubic>,<at 0.6412 0.2232 RCIM_Cubic>]}
         GreenCurve {keys [<at 0 0.8969 RCIM_Cubic>,<at 0.6412 0.7305 RCIM_Cubic>]}
         RedCurve {keys [<at 0 0.6227 RCIM_Cubic>,<at 0.6412 0.0395 RCIM_Cubic>]}
        }}))
    )
    ScaleA=true
    ScaleRGB=true
   )
   &ScaleRibbonWidth(Ribbon Width Scale=&FloatFromCurve(FloatCurve=DataInterface() {propertyValues {Curve {keys [<at 0 1 RCIM_Cubic>,<at 1 0.3 RCIM_Cubic>]}}}))
   /SetVariables(Particles.Position=&AddVectorToPosition(
     Position=&FN_FrontEndVector3Modifier(
      Gameplay=<0 0 0>
      Front End=&MakeVector(
       X=0
       Y=0
       Z=&FloatFromCurve(FloatCurve=DataInterface() {propertyValues {Curve {keys [
            <at 0 0 RCIM_Cubic>
            {arriveTangent=21.37, interpMode=RCIM_Cubic, leaveTangent=21.37, time=0.214, value=3.2426}
            {arriveTangent=11.6337, interpMode=RCIM_Cubic, leaveTangent=11.6337, time=0.31, value=6.6247}
            {arriveTangent=-0.1013, interpMode=RCIM_Cubic, leaveTangent=-0.1013, tangentMode=RCTM_User, time=0.6229, value=8}
            {arriveTangent=-22.9955, interpMode=RCIM_Cubic, leaveTangent=-22.9955, tangentMode=RCTM_User, time=0.8, value=0}
           ]}}})
      )
     )
     Vector=&Lerp_Vector(
      A=Linked() {linkedVariable {name=Particles.SkeletalMeshLocation.SampledPosition, type=NiagaraPosition}}
      B=&FN_FrontEndVector3Modifier(Gameplay=Linked() {linkedVariable {name=Particles.SkeletalMeshLocation001.SampledPosition, type=NiagaraPosition}}, Front End=<25 0 30>)
      Alpha=$Particles.NormalizedAge
     )
    )) {isSetParameters=true}
  ]
  rendererClasses [&NiagaraRibbonRendererProperties]
  renderers [{rendererClass=&NiagaraRibbonRendererProperties}]
 }
 Splashes001 {
  emitterUpdateScript [
   &EmitterState(Life Cycle Mode=::NewEnumerator1, Loop Behavior=::NewEnumerator1, Loop Delay=8.15, Loop Duration=0.24, UseLoopDelay=true)
   &SpawnRate(SpawnRate=&FN_FrontEndFloatModifier(Gameplay=12, Front End=24), Use Spawn Probability=false)
   &SpawnBurst_Instantaneous(Spawn Count=2, Use Spawn Probability=false, UseLoopCountLimit=false)
  ]
  particleSpawnScript [
   &InitializeParticle(
    Color=&MakeLinearColorFromVectorAndFloat(
     "Vector (RGB)"=&FN_FrontEndVector3Modifier(
      Gameplay=&MakeVectorFromLinearColorRGB(Color=&RandomRangeLinearColor(Minimum=<rgba 0.0773 0.9331 1 1>, Maximum=<rgba 0.0838 1 0.5546 1>))
      Front End=&MakeVectorFromLinearColorRGB(Color=&RandomRangeLinearColor(Minimum=<rgba 0.1251 0.7305 0.0647 1>, Maximum=<rgba 0.0304 0.7305 0.2165 1>))
     )
     "Float (Alpha)"=1
    )
    Lifetime Max=0.3
    Lifetime Min=0.125
    Lifetime Mode=::NewEnumerator1
    Sprite Rotation Mode=::NewEnumerator2
    Sprite Size Mode=::NewEnumerator1
    Uniform Sprite Size Max=16
    Uniform Sprite Size Min=12
    UsePositionOffset=false
    Write Material Random=true
   )
   &SocketLocation(SamplePosOffset=false, Socket Index=1, Socket Reader=$System.Socket Reader, Socket Sampling Mode=::NewEnumerator3, VelocitySpeedLimit=false) {enabled=false}
   &SkeletalMeshLocation(
    Bone / Socket Index=1
    Bone Sampling Mode=::NewEnumerator4
    MaskSpawnGroup=false
    Sampled Position Offset=&FN_FrontEndVector3Modifier(Gameplay=<-4 4 -2>, Front End=<0 0 0>)
    Sampled Position Offset Coordinate Space:ENiagaraSkelMeshTransforms=::NewEnumerator2
    Skeletal Mesh=$System.Skeletal Mesh Reader
   ) {module=SkeletalMeshLocation001}
   &ShapeLocation(
    MaskSpawnGroup=false
    Offset=<-2 4 -8>
    Offset Mode=::NewEnumerator0
    Override Seed=false
    Shape Origin=$Particles.SkeletalMeshLocation001.SampledPosition
    Sphere Radius=2
    UseOwnerScale=true
   )
   &AddVelocity(
    Module.Use Falloff Distance=false
    Origin Offset=<0 0 -1>
    Override Seed=false
    Velocity Mode=::NewEnumerator1
    Velocity Origin=$Particles.SkeletalMeshLocation001.SampledPosition
    Velocity Speed=&RandomRangeFloat(Minimum=5, Maximum=12)
   )
   &SubUVAnimation(
    "Sprite Renderer "=DataInterface() {propertyValues {SpriteRenderer=&self:Splashes001_0.NiagaraSpriteRendererProperties_0}}
    SubUV Animation Mode:ENiagara_SubUVLookupModeV2=::NewEnumerator2
    UseEndFrame=false
    UseStartFrame=false
   )
   &CameraOffset(Camera Offset Amount=&FN_FrontEndFloatModifier(Gameplay=0, Front End=&RandomRangeFloat(Minimum=4, Maximum=12)))
  ]
  particleUpdateScript [
   &ParticleState()
   &ScaleSpriteSize(Uniform Curve Sprite Scale=DataInterface() {propertyValues {Curve {
       defaultValue=0
       keys [
        {arriveTangent=9.0725, interpMode=RCIM_Cubic, leaveTangent=9.0725, tangentMode=RCTM_User, time=0, value=0}
        {arriveTangent=0.2183, interpMode=RCIM_Cubic, leaveTangent=0.2183, tangentMode=RCTM_User, time=0.2999, value=1.0751}
       ]
      }}})
   &GravityForce(Gravity=<0 0 -120>)
   &SolveForcesAndVelocity(Clamp Velocity=false, Limit Acceleration=false)
   &ScaleSpriteSizeBySpeed(Max Scale Factor=<2 2.2>, Min Scale Factor=<1 1.2>)
   &DynamicMaterialParameters(
    Index 0 Param 4=&FloatFromCurve(FloatCurve=DataInterface() {propertyValues {Curve {
        defaultValue=0.9847
        keys [{interpMode=RCIM_Cubic, tangentMode=RCTM_User, time=0, value=0}, {interpMode=RCIM_Cubic, tangentMode=RCTM_User, time=1, value=1}]
       }}})
    Module.Index0Float0Write=true
    Module.Index0Float1Write=true
    Module.Index0Float2Write=true
    Module.Index0Float3Write=true
   )
   &ScaleColor(
    Scale RGB=&Multiply_VectorByFloat(
     Vector=&MakeVectorFromLinearColorRGB(Color=&FNTimeOfDayColor(Night Color=<rgba 0.3229 0.661 1 1>, Morning Color=<rgba 1 0.7745 0.7008 1>, Day Color=<rgba 1 1 1 1>, Evening Color=<rgba 1 1 1 1>))
     Float=&RandomRangeFloat(Minimum=1.8, Maximum=2.2)
    )
    ScaleA=true
    ScaleRGB=true
   )
  ]
  rendererClasses [&NiagaraSpriteRendererProperties]
  renderers [{rendererClass=&NiagaraSpriteRendererProperties}]
 }
}
summary {
 systemName=NS_Emote_AboutJury
 emitters [
  {emitterName=SlurpJuice, rendererClasses [&NiagaraRibbonRendererProperties]}
  {emitterName=Confetti001, rendererClasses [&NiagaraSpriteRendererProperties]}
  {emitterName=Liquid001, rendererClasses [&NiagaraMeshRendererProperties]}
  {emitterName=GlassDrops001, rendererClasses [&NiagaraSpriteRendererProperties]}
  {emitterName=Splashes001, rendererClasses [&NiagaraSpriteRendererProperties]}
  {emitterName=Flare, rendererClasses []}
 ]
 userVariables [
  User.bIsFrontEnd {type=NiagaraBool}
  User.bIsFrontEndPreview {type=NiagaraBool}
  User.CrowdCheerCurve {type=NiagaraFloat, defaultValue=0, description=""}
  User.LiquidScale {type=NiagaraFloat, defaultValue=0.9, description=""}
 ]
}
